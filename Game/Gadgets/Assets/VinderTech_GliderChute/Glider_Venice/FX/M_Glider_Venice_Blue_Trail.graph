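# Unlit two-sided translucent material for the blue trail effect of the Venice glider, usable on Niagara ribbons and mesh particles.
# World position offset combines sine waves with SplineThicken; emissive is particle color scaled by a U-axis gradient.
# Opacity layers a panning, noise-distorted trail texture, noise dissolve and fades for view angle, trail ends and depth.
# Niagara dynamic parameters drive width, UV offset, V flip, panning, desaturation and appearance.
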Material /Game/Gadgets/Assets/VinderTech_GliderChute/Glider_Venice/FX/M_Glider_Venice_Blue_Trail
graph {
 properties {
  bUsedWithNiagaraMeshParticles=true
  bUsedWithNiagaraRibbons=true
  blendMode=BLEND_Translucent
  pixelDepthOffsetMode=PDOM_Legacy
  shadingModel=MSM_Unlit
  twoSided=true
  usedShadingModels=Unlit
 }
 nodes [
  n1 = Add(
   A=TextureCoordinate()
   B=AppendVector(A=DynamicParameter() {paramNames [Width and AngleFadeScale, Offset, V_filp, Appear]}.Offset, B=Constant())
  )
  n2 = Time()
  n3 = Sine(None=Add(A=ComponentMask(None=@n1) {r=true}, B=Multiply(A=@n2) {constB=-0.8}))
  n4 = Constant3Vector() {constant=<rgba 0 0 1 0>}
  n5 = Multiply(A=Transform(None=Multiply(A=@n3, B=@n4)) {transformSourceType=TRANSFORMSOURCE_Local})
  n6 = Constant3Vector() {constant=<rgba 0 1 0 0>}
  n7 = Multiply(
   A=Transform(None=Multiply(A=@n3, B=@n6)) {transformSourceType=TRANSFORMSOURCE_Local}
   B=ScalarParameter() {defaultValue=10, parameterName=WPO Intensity}
  ) {constB=8}
  n8 = TextureCoordinate()
  n9 = Saturate(None=Multiply(A=ComponentMask(None=@n8) {r=true}) {constB=3})
  n10 = DynamicParameter() {paramNames [Width and AngleFadeScale, Offset, V_filp, Appear]}
  n11 = Multiply(A=@n10.Width and AngleFadeScale, B=ScalarParameter() {defaultValue=120, parameterName=WidthBase})
  n12 = Multiply(A=@n10.Width and AngleFadeScale, B=ScalarParameter() {defaultValue=20, parameterName=WidthTip})
  n13 = Power(Base=ComponentMask(None=TextureCoordinate()) {r=true}) {constExponent=0.7}
  n14 = StaticBool()
  Add(
   A=Multiply(A=Add(A=@n5, B=@n7), B=LinearInterpolate(Alpha=@n9))
   B=LinearInterpolate(
    B=MaterialFunctionCall(WidthBase=@n11, WidthTip=@n12, UVs for Thickness=@n13, Expand U or V UV Channel=@n14) {materialFunction=&SplineThicken}.WorldPositionOffset
    Alpha=ScalarParameter() {defaultValue=1, parameterName=Use SplineThicken}
   )
  )
  n15 = ComponentMask(None=TextureCoordinate()) {r=true}
  n16 = LinearInterpolate(
   A=ScalarParameter() {defaultValue=1, parameterName=Color Transition Min}
   Alpha=Power(Base=OneMinus(None=@n15), Exp=ScalarParameter() {defaultValue=1.5, parameterName=Color Transition power}) {constExponent=1.5}
  )
  n17 = ParticleColor()
  MaterialFunctionCall(Emissive=Desaturation(
    None=Multiply(A=@n16, B=@n17.RGB)
    Fraction=DynamicParameter() {parameterIndex=1, defaultValue=<rgba 0 0 0 1>, paramNames [PanX, PanY, Des, AngleFadeScale]}.Des
   )) {materialFunction=&MF_FX_EmissiveAdjustments}
  n18 = Multiply(A=ComponentMask(None=TextureCoordinate()) {r=true}) {constB=10}
  n19 = ScalarParameter() {parameterName="FadeSensitivity (optional) (S)"}
  n20 = Multiply(A=MaterialFunctionCall() {materialFunction=&FadeBasedOnViewAngle}.Result) {constB=10}
  n21 = DynamicParameter() {parameterIndex=1, defaultValue=<rgba 0 0 0 1>, paramNames [PanX, PanY, Des, AngleFadeScale]}
  n22 = MaterialFunctionCall(X=@n21.PanX, Y=@n21.PanX) {materialFunction=&MakeFloat2}
  n23 = VectorParameter() {parameterName=UV_Scale, defaultValue=<rgba 0.9 1 0 0>}
  n24 = Multiply(
   A=Multiply(A=TextureCoordinate(), B=ComponentMask(None=@n23.RGB) {g=true, r=true})
   B=AppendVector(A=Constant() {r=1}, B=DynamicParameter() {paramNames [Width and AngleFadeScale, Offset, V_filp, Appear]}.V_filp)
  )
  n25 = LinearInterpolate(
   B=AppendVector(A=DynamicParameter() {paramNames [Width and AngleFadeScale, Offset, V_filp, Appear]}.Offset, B=Constant())
   Alpha=ScalarParameter() {defaultValue=1, parameterName=Use Default Tex Offset}
  )
  n26 = TextureCoordinate() {uTiling=0.7, vTiling=0.2}
  n27 = AppendVector(A=DynamicParameter() {paramNames [Width and AngleFadeScale, Offset, V_filp, Appear]}.Offset, B=Constant())
  n28 = TextureSampleParameter2D(UVs=Panner(Coordinate=Add(A=@n26, B=@n27)) {speedX=-0.3}) {parameterName=T_Distortion_01, samplerType=SAMPLERTYPE_LinearColor, texture=/Game/Effects/Fort_Effects/BluGlow/Violentium/Violentium_Mask}
  n29 = Multiply(A=@n28.R, B=ScalarParameter() {defaultValue=0.08, parameterName=Default Tex Disturbed}) {constB=0.08}
  n30 = ScalarParameter() {defaultValue=1, parameterName=Distortion_Macro_Tile_X}
  n31 = ScalarParameter() {defaultValue=1, parameterName=Distortion_Macro_Tile_Y}
  n32 = Add(A=Multiply(A=AppendVector(A=@n30, B=@n31), B=TextureCoordinate() {vTiling=0.7}), B=Constant2Vector() {g=0.2})
  n33 = AppendVector(A=DynamicParameter() {paramNames [Width and AngleFadeScale, Offset, V_filp, Appear]}.Offset, B=Constant())
  n34 = TextureSampleParameter2D(UVs=Panner(Coordinate=Add(A=@n32, B=@n33)) {speedX=-1.5}) {
   parameterName=T_Distortion_02
   samplerType=SAMPLERTYPE_Normal
   texture=/Game/Characters/Player/Female/Medium/Bodies/F_MED_Bistro_Astronaut/FX/T_GaussianNoise_Normal
  }
  n35 = ComponentMask(None=TextureCoordinate()) {r=true}
  n36 = Multiply(
   A=Multiply(A=@n34.R, B=ScalarParameter() {defaultValue=0.5, parameterName=Noise Intensity}) {constB=0.5}
   B=Saturate(None=Power(Base=@n35) {constExponent=1.5})
  )
  n37 = ScalarParameter() {parameterName=Use Top Disturbed}
  n38 = Panner(
   Coordinate=Add(A=Add(A=@n24, B=@n25), B=LinearInterpolate(A=@n29, B=@n36, Alpha=@n37)) {constB=0}
   Speed=ComponentMask(None=VectorParameter() {parameterName=Speed, defaultValue=<rgba -1 0 0 0>}.RGB) {g=true, r=true}
  ) {speedX=-1}
  n39 = StaticComponentMaskParameter(None=TextureSampleParameter2D(UVs=Add(A=@n22.Result, B=@n38)) {parameterName=Default_Tex, texture=&/T_Glider_Venice_Trail}.RGB) {defaultR=true, parameterName=RGB Mask}
  n40 = TextureCoordinate() {uTiling=1.5}
  n41 = Constant2Vector() {g=0.2}
  n42 = TextureSample(UVs=Panner(Coordinate=Add(A=@n40, B=@n41)) {speedX=-1.5}) {samplerType=SAMPLERTYPE_Grayscale, texture=&LowResBlurredNoise}
  n43 = LinearInterpolate(Alpha=MaterialFunctionCall() {materialFunction=&DebugTimeSine}.Result) {constA=-0.2, constB=0}
  n44 = TextureCoordinate()
  n45 = Multiply(A=Add(A=ComponentMask(None=@n44) {r=true}) {constB=-0.5})
  n46 = Power(
   Base=LinearInterpolate(A=@n42.R, B=LinearInterpolate(A=@n43, Alpha=@n42.R), Alpha=ScalarParameter() {parameterName=Use Random Dissolution})
   Exp=LinearInterpolate(
    A=ScalarParameter() {defaultValue=0.1, parameterName=Dissolve mask Min}
    B=ScalarParameter() {defaultValue=2, parameterName=Dissolve mask Max}
    Alpha=Saturate(None=@n45)
   ) {constA=0.1, constB=2}
  )
  n47 = LinearInterpolate(Alpha=@n46) {constA=-1, constB=1.5}
  n48 = LinearInterpolate(
   B=LinearInterpolate(A=Round(None=@n46), B=Saturate(None=@n47), Alpha=ScalarParameter() {parameterName=Use Soft Dissolve})
   Alpha=ScalarParameter() {defaultValue=1, parameterName=Use dissolution}
  ) {constA=1}
  n49 = Multiply(
   A=Multiply(A=@n17.A, B=Multiply(A=@n39, B=@n48))
   B=LinearInterpolate(
    A=MaterialFunctionCall("FadeSensitivity (optional)"=@n19) {materialFunction=&FadeBasedOnViewAngle}.Result
    B=Saturate(None=@n20)
    Alpha=ScalarParameter() {parameterName=Use Angle Fade}
   )
  )
  n50 = TextureCoordinate()
  n51 = LinearInterpolate(Alpha=OneMinus(None=ComponentMask(None=@n50) {r=true})) {constA=-0.5}
  n52 = LinearInterpolate(Alpha=DynamicParameter() {paramNames [Width and AngleFadeScale, Offset, V_filp, Appear]}.Appear) {constA=-1, constB=2}
  Multiply(A=Multiply(A=Saturate(None=@n18), B=@n49), B=Saturate(None=Add(A=@n51, B=@n52)))
  Multiply(
   A=Multiply(A=Constant() {r=5}, B=DynamicParameter() {paramNames [Width and AngleFadeScale, Offset, V_filp, Appear]}.Width and AngleFadeScale)
   B=VertexColor().R
  )
  n53 = DepthFade(Opacity=@n49, FadeDistance=ScalarParameter() {defaultValue=20, parameterName=FadeDistance})
  ShadingPathSwitch(Default=@n53, Deferred=@n53, Forward=@n49, Mobile=@n49)
 ]
}
